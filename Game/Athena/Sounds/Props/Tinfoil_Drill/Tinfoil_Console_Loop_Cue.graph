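# Looping sound cue for the Tinfoil Drill console prop, assigned to the Fort_Gadgets sound class.
# A mixer layers two looping waves at equal volume, with the Loop_01 layer passing through a modulator that applies a fixed upward pitch shift.
# The cue raises overall pitch, lowers volume and overrides attenuation for a short-range falloff with slight stereo spread.
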SoundCue /Game/Athena/Sounds/Props/Tinfoil_Drill/Tinfoil_Console_Loop_Cue
root=Mixer_2
nodes [
 {id=Mixer_2, children [Modulator_1,WavePlayer_5], properties {inputVolume [1,1]}}
 {id=Modulator_1, children [WavePlayer_6], properties {pitchMax=1.1, pitchMin=1.1}}
 {id=WavePlayer_5, properties {bLooping=true, soundWaveAssetPtr=&/Tinfoil_Console_Loop_04}}
 {id=WavePlayer_6, properties {bLooping=true, soundWaveAssetPtr=&/Tinfoil_Console_Loop_01}}
]
properties {
 bOverrideAttenuation=true
 maxDistance=550
 pitchMultiplier=1.1
 soundClassObject=&Fort_Gadgets
 volumeMultiplier=0.5
 attenuationOverrides {falloffDistance=500, stereoSpread=20, attenuationShapeExtents {x=50}}
}
