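# Sound cue for the activation of the timer on the Tinfoil Drill console prop in Athena.
# A single wave player plays one activation sample directly, with no randomization or modulation.
# The cue is routed to the gadgets sound class at reduced volume and overrides attenuation with its own limited falloff range and stereo spread.
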SoundCue /Game/Athena/Sounds/Props/Tinfoil_Drill/Tinfoil_Console_Timer_Activate_Cue
root=WavePlayer_2
nodes [{id=WavePlayer_2, properties {soundWaveAssetPtr=&/Tinfoil_Console_Timer_Activate_02}}]
properties {
 bOverrideAttenuation=true
 maxDistance=1400
 soundClassObject=&Fort_Gadgets
 volumeMultiplier=0.4
 attenuationOverrides {falloffDistance=1200, stereoSpread=500, attenuationShapeExtents {x=200}}
}
