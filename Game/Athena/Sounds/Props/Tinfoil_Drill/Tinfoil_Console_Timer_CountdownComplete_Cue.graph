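# Sound cue for the Tinfoil Drill console prop, named for the moment its timer countdown completes.
# A single wave player feeds the countdown-complete wave straight to the output, with no randomization or modulation.
# The cue sits in the gadgets sound class at reduced volume and overrides attenuation for a limited audible range, custom falloff and stereo spread.
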SoundCue /Game/Athena/Sounds/Props/Tinfoil_Drill/Tinfoil_Console_Timer_CountdownComplete_Cue
root=WavePlayer_0
nodes [{id=WavePlayer_0, properties {soundWaveAssetPtr=&/Tinfoil_Console_Timer_CountdownComplete_01}}]
properties {
 bOverrideAttenuation=true
 maxDistance=1400
 soundClassObject=&Fort_Gadgets
 volumeMultiplier=0.35
 attenuationOverrides {falloffDistance=1200, stereoSpread=500, attenuationShapeExtents {x=200}}
}
