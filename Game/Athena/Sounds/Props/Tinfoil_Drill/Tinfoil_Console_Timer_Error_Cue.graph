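# Sound cue for the error sound of the Tinfoil console timer, filed under the Tinfoil Drill prop audio.
# A single wave player feeds one error wave straight to the output at reduced volume in the Gadgets sound class.
# Inline attenuation keeps it local to the source with some stereo spread.
# Inline concurrency allows one instance, rejects new ones while it plays and sets a retrigger time to limit rapid repeats.
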SoundCue /Game/Athena/Sounds/Props/Tinfoil_Drill/Tinfoil_Console_Timer_Error_Cue
root=WavePlayer_1
nodes [{id=WavePlayer_1, properties {soundWaveAssetPtr=&/Tinfoil_Console_Timer_Error_01}}]
properties {
 bOverrideAttenuation=true
 bOverrideConcurrency=true
 maxDistance=1400
 soundClassObject=&Fort_Gadgets
 volumeMultiplier=0.2
 attenuationOverrides {falloffDistance=1200, stereoSpread=500, attenuationShapeExtents {x=200}}
 concurrencyOverrides {maxCount=1, resolutionRule=PreventNew, retriggerTime=2, voiceStealReleaseTime=2}
}
