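# Sound cue for the timer tick of the Tinfoil drill console prop.
# A mixer layers two tick wave players at once, with the first slightly lower in level than the second.
# The cue plays quietly under the gadgets sound class, using its own attenuation override with a short falloff distance and stereo spread.
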SoundCue /Game/Athena/Sounds/Props/Tinfoil_Drill/Tinfoil_Console_Timer_Tick_Cue
root=Mixer_0
nodes [
 {id=Mixer_0, children [WavePlayer_0,WavePlayer_1], properties {inputVolume [0.7,1]}}
 {id=WavePlayer_0, properties {soundWaveAssetPtr=&/Tinfoil_Console_Timer_Tick_01}}
 {id=WavePlayer_1, properties {soundWaveAssetPtr=&/Tinfoil_Console_Timer_Tick_02}}
]
properties {
 bOverrideAttenuation=true
 maxDistance=800
 soundClassObject=&Fort_Gadgets
 volumeMultiplier=0.15
 attenuationOverrides {falloffDistance=800, stereoSpread=500, attenuationShapeExtents {x=0}}
}
