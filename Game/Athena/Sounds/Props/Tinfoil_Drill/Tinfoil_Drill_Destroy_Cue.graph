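# Destruction sound cue for the Tinfoil Drill prop, built from reused rock, metal and vehicle impact waves.
# A quality-level switch gives its first two slots a full mix: pitch-varied random rock hit, delayed random metal impact, delayed hoverboard explosion and two layered Hamster destroy waves.
# The last two slots keep only the metal and Hamster layers.
# A root modulator feeds the impacts class, with shopping-cart attenuation and capped concurrency.
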SoundCue /Game/Athena/Sounds/Props/Tinfoil_Drill/Tinfoil_Drill_Destroy_Cue
root=Modulator_7
nodes [
 {id=Delay_0, children [Random_4], properties {delayMax=0.2, delayMin=0.2}}
 {id=Delay_3, children [Random_6], properties {delayMax=0.2, delayMin=0.2}}
 {id=Delay_4, children [Modulator_5], properties {delayMax=0.1, delayMin=0.1}}
 {id=Mixer_1, children [Modulator_0,Delay_0,Delay_4,Mixer_4], properties {inputVolume [0.7,0.35,1.2,0.7]}}
 {id=Mixer_4, children [WavePlayer_16, WavePlayer_17], properties {inputVolume [1,1]}}
 {id=Mixer_5, children [Delay_3,Mixer_6], properties {inputVolume [0.5,1]}}
 {id=Mixer_6, children [WavePlayer_27, WavePlayer_28], properties {inputVolume [1,1]}}
 {id=Modulator_0, children [Random_0], properties {pitchMax=1.3}}
 {id=Modulator_5, children [WavePlayer_14]}
 {id=Modulator_7, children [QualityLevel_1]}
 {id=QualityLevel_1, children [Mixer_1,Mixer_1,Mixer_5,Mixer_5]}
 {id=Random_0, children [WavePlayer_4,WavePlayer_0,WavePlayer_6], properties {weights [1,1,1]}}
 {id=Random_4, children [WavePlayer_20, WavePlayer_21], properties {weights [1,1]}}
 {id=Random_6, children [WavePlayer_29, WavePlayer_30], properties {weights [1,1]}}
 {id=WavePlayer_0, properties {soundWaveAssetPtr=/Game/Sounds/Fort_Impact_Sounds/Demolition/Rock/Rock_Ore_Impact_02}}
 {id=WavePlayer_14, properties {soundWaveAssetPtr=/Game/Athena/Sounds/Vehicles/Hoverboard/Hoverboard_Explode_01}}
 {id=WavePlayer_16, properties {soundWaveAssetPtr=/Game/Athena/Sounds/Vehicles/Hamster/Hamster_Impact_Destroy_02}}
 {id=WavePlayer_17, properties {soundWaveAssetPtr=/Game/Athena/Sounds/Vehicles/Hamster/Hamster_Impact_Destroy_01}}
 {id=WavePlayer_20, properties {soundWaveAssetPtr=/Game/Sounds/Fort_Impact_Sounds/Demolition/Metal/Metal_Machine_ImpactFinal03}}
 {id=WavePlayer_21, properties {soundWaveAssetPtr=/Game/Sounds/Fort_Impact_Sounds/Demolition/Metal/Metal_Machine_ImpactFinal02}}
 {id=WavePlayer_27, properties {soundWaveAssetPtr=&Hamster_Impact_Destroy_02}}
 {id=WavePlayer_28, properties {soundWaveAssetPtr=&Hamster_Impact_Destroy_01}}
 {id=WavePlayer_29, properties {soundWaveAssetPtr=&Metal_Machine_ImpactFinal03}}
 {id=WavePlayer_30, properties {soundWaveAssetPtr=&Metal_Machine_ImpactFinal02}}
 {id=WavePlayer_4, properties {soundWaveAssetPtr=/Game/Sounds/Fort_Impact_Sounds/Demolition/Rock/Rock_Ore_Impact_01}}
 {id=WavePlayer_6, properties {soundWaveAssetPtr=/Game/Sounds/Fort_Impact_Sounds/Demolition/Rock/Rock_Ore_Impact_03}}
]
properties {
 attenuationSettings=/Game/Sounds/Attenuation/Fort_ShoppingCart_Movement_Att
 maxDistance=5300
 soundClassObject=&Fort_Impacts
 volumeMultiplier=1
 attenuationOverrides {reverbWetLevelMax=0.7, reverbWetLevelMin=0.2}
 concurrencyOverrides {resolutionRule=PreventNew}
 concurrencySet [&CONC_Object_Impacts_FourMax]
}
