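# Sound cue for the looping sound of the Tinfoil Drill prop in Athena.
# A single looping wave player feeds the output directly, with no randomization or mixing.
# The cue routes to the Fort_Gadgets sound class at slightly reduced volume and overrides attenuation inline, setting its own falloff distance, inner radius and wide stereo spread.
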SoundCue /Game/Athena/Sounds/Props/Tinfoil_Drill/Tinfoil_Drill_Loop_Cue
root=WavePlayer_1
nodes [{id=WavePlayer_1, properties {bLooping=true, soundWaveAssetPtr=&/Tinfoil_Drill_Loop_01}}]
properties {
 bOverrideAttenuation=true
 maxDistance=2300
 soundClassObject=&Fort_Gadgets
 volumeMultiplier=0.8
 attenuationOverrides {falloffDistance=1700, stereoSpread=3000, attenuationShapeExtents {x=600}}
}
