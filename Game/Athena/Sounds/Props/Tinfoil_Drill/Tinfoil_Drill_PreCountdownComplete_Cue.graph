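# Sound cue for the Tinfoil Drill prop, named for the completion of its pre-countdown.
# It plays a single wave player directly, with no randomization or mixing, on the Fort_Gadgets sound class at slightly reduced volume.
# Inline attenuation overrides replace any shared attenuation asset, setting the falloff, inner shape extent and stereo spread.
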SoundCue /Game/Athena/Sounds/Props/Tinfoil_Drill/Tinfoil_Drill_PreCountdownComplete_Cue
root=WavePlayer_0
nodes [{id=WavePlayer_0, properties {soundWaveAssetPtr=&/Tinfoil_Drill_PreCountdownComplete_01}}]
properties {
 bOverrideAttenuation=true
 maxDistance=2400
 soundClassObject=&Fort_Gadgets
 volumeMultiplier=0.9
 attenuationOverrides {falloffDistance=1700, stereoSpread=3000, attenuationShapeExtents {x=700}}
}
